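; Battle pass S20 quest ability for a two-tap with the single break-action shotgun, triggered by a gameplay event.
; After the parent call, it checks the event's instigator tags for that shotgun and its target tags for an Athena player.
; On a match, it applies its configured gameplay effect to the owner and waits for a Damage.Received event to remove it by handle.
; EndAbility follows the branch rather than the event callback.
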(event Ability|EventActivateAbilityFromEvent
  (params (EventData :type Gameplay%20Event%20Data%20Structure%20%28by%20ref%29))
  (bind (_eventtag _instigator _target _optionalobject _optionalobject2 _contexthandle _instigatortags _targettags) (Utilities|Struct|BreakGameplayEventData EventData))
  (|CallParentFunction EventData)
  (if (and (GameplayTags|HasTag _instigatortags "(TagName=\"Item.Weapon.Ranged.Shotgun.SingleBreakAction\")" true) (GameplayTags|HasTag _targettags "(TagName=\"Pawn.Athena.Player\")"))
    (bind _returnvalue (Ability|ApplyGameplayEffectToOwner (Variables|Default|GetGameplay%20Effect%20Class)))
    (Ability|Tasks|WaitGameplayEvent "(TagName=\"Event.Damage.Received\")"
      (:EventReceived
        (Ability|RemoveGameplayEffectFromOwnerWithHandle _returnvalue)))
    (else
      _))
  (Ability|EndAbility))
