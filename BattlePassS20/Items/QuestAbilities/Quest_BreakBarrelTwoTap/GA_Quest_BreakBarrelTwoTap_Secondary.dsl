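; Quest gameplay ability for the secondary part of the Break Barrel Two Tap battle pass quest, activated by a gameplay event.
; After calling the parent, it checks that the event's instigator tags exactly match the single break-action shotgun and its target tags include a player pawn.
; On a match, it sends the BreakbarrelTwoTap custom stat event, count one, to the owning player's quest manager.
; The ability ends whether or not the check passes.
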(event Ability|EventActivateAbilityFromEvent
  (params (EventData :type Gameplay%20Event%20Data%20Structure%20%28by%20ref%29))
  (bind (_eventtag _instigator _target _optionalobject _optionalobject2 _contexthandle _instigatortags _targettags) (Utilities|Struct|BreakGameplayEventData EventData))
  (|CallParentFunction EventData)
  (if (and (GameplayTags|HasTag _instigatortags "(TagName=\"Item.Weapon.Ranged.Shotgun.SingleBreakAction\")" true) (GameplayTags|HasTag _targettags "(TagName=\"Pawn.Athena.Player\")"))
    (Quest|SendComplexCustomStatEvent (Variables|GAVars|Data|GetOwningPlayerQuestManager) 0 "(GameplayTags=((TagName=\"Athena.Quests.BreakbarrelTwoTap\")))" "(GameplayTags=)" 1)
    (else
      _))
  (Ability|EndAbility))
